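# Landscape material function that adds Helios-style snow to roads, toggled by a static switch.
# It height-blends the input material with a snow layer, using road-edge sphere masks, world-space noise and an optional squared biome mask texture.
# Separate outputs supply post-RVT mask, base color and emissive from a second snow function.
# MaterialProxyReplace substitutes a tinted snow base color for proxy materials.
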MaterialFunction /Game/Athena/Apollo/Environments/Landscape/Materials/MaterialFunctions/MF_HermesSnow
graph {nodes [
  n1 = StaticBool()
  n2 = MaterialFunctionCall!(WorldPosition=StaticSwitch(
    True=MaterialFunctionCall() {materialFunction=&WorldPositionBehindTranslucency}.Result
    False=WorldPosition().XYZ
    Value=@n1
   )) {materialFunction=/Game/Athena/Helios/Environments/Landscape/Materials/MaterialFunctions_Utilities/MF_TerrainCoordinates}
  Composite!(InUVs=@n2.Adjustable, Enabled=@n1) {subgraphName=DerrivePageMask}
  Constant() {r=1.4}
  Constant() {r=2}
  FunctionInput() {bUsePreviewValueAsDefault=true, inputName=-----------------------, inputType=FunctionInput_Scalar, sortPriority=99}
  n3 = NamedRerouteUsage()
  n4 = Add(A=OneMinus(None=NamedRerouteUsage()), B=@n3)
  n5 = MaterialFunctionCall!(
   Material 1=NamedRerouteUsage()
   Height Texture 1=NamedRerouteUsage()
   Material 2=NamedRerouteUsage()
   Height Texture 2=NamedRerouteUsage()
   Transistion Phase=StaticSwitchParameter(True=Saturate(None=@n4), False=@n3) {defaultValue=true, group=HeliosSnow, parameterName=Use_SnowBiomeMaskTexture, sortPriority=-800}
   UseContrast
   Contrast
   AddLayerMask
  ) {materialFunction=/Game/Athena/Helios/Environments/Landscape/Materials/MaterialFunctions_Utilities/MF_HeightBlendLandscapeLayers}
  FunctionOutput(None=@n5.Alpha) {outputName=PostRVT_Mask, sortPriority=100}
  n6 = FunctionInput() {bUsePreviewValueAsDefault=true, inputName=PostRVT_READ_Material, inputType=FunctionInput_MaterialAttributes, sortPriority=100}
  n7 = GetMaterialAttributes(None=MaterialFunctionCall!(
    ScalarWeightMask
    UseSnowOcclusionTextureMask
    ------
    UseMaterialInput
    AdditionalMaskToRemoveSnow
    OutputTangentNormals
    UseAdditionalMaskToRemoveSnow
    UseSnowNormals
    OuputVertexNormalsAsTheSnowNormals
    UseSparkle
    LQSpark
    AllowSnowMicroNormals
    CliffMaterialOnly
    Material=@n6
    CliffLandscapeLayerMask
    IsLandscape=StaticBool()
    AreInputNormalsTangentSpace=StaticBool()
    UseInputNormalsForFakeSnowShadow=StaticBool() {value=true}
   ) {materialFunction=/Game/Athena/Helios/Environments/Landscape/Materials/QualityLevelLayers/MF_Helios_Snow_Default}.SnowMaterialOnly) {attributeGetTypes [#1, #2]}
  FunctionOutput(None=@n7.BaseColor) {outputName=PostRVT_Basecolor, sortPriority=105}
  FunctionOutput(None=@n7.EmissiveColor) {outputName=PostRVT_Emissive, sortPriority=106}
  n8 = NamedRerouteUsage()
  n9 = BlendMaterialAttributes(A=NamedRerouteUsage(), B=@n8, Alpha=NamedRerouteUsage()) {vertexAttributeBlendType=UseA}
  FunctionOutput(None=StaticSwitchParameter(
    True=StaticSwitchParameter(True=@n9, False=@n8) {defaultValue=true, group=HeliosSnow, parameterName=Use_SnowBiomeMaskTexture, sortPriority=-800}
    False=NamedRerouteUsage()
   ) {group=HeliosSnow, parameterName=UseHelios_RoadSnow, sortPriority=-5000})
  MaterialFunctionCall!(
   RVT=MaterialFunctionCall(NoRVT=NamedRerouteUsage(), RVT=NamedRerouteUsage()) {materialFunction=&FN_RVT_Switch_Material}.Result
   ByPass RVT=@n8
  ) {materialFunction=/Game/Athena/Apollo/Maps/Landscape/VT/MaterialFunctions/EnableRVT}
  n10 = GetMaterialAttributes(None=FunctionInput() {bUsePreviewValueAsDefault=true, inputName=Material, inputType=FunctionInput_MaterialAttributes, sortPriority=-500}) {attributeGetTypes [#3]}
  NamedRerouteDeclaration(None=@n10.MaterialAttributes) {name=Input_Material, nodeColor=<rgba 0 1 0 1>}
  n11 = MaterialFunctionCall!(Grass_UseCustomSlopeMask, Grass_CustomSlopeMask) {materialFunction=/Game/Athena/Helios/Environments/Landscape/Materials/MaterialFunctions_Utilities/MF_Helios_StaticMeshLandscapeLayerMatch}
  n12 = VectorParameter() {group=HeliosSnow, parameterName=RoadSnow_ProxybaseColorAdjustment, defaultValue=<rgba 0.2747 0.3916 0.4793 1>}
  n13 = GetMaterialAttributes(None=MaterialProxyReplace(
    Realtime=@n11.SnowLayer
    MaterialProxy=SetMaterialAttributes!(MaterialAttributes=@n11.SnowLayer, BaseColor=@n12.RGB) {attributeSetTypes [#1]}
   )) {attributeGetTypes [#3]}
  NamedRerouteDeclaration(None=@n13.MaterialAttributes) {name=Snow_Material, nodeColor=<rgba 1 1 1 1>}
  NamedRerouteDeclaration(None=MaterialFunctionCall(
    Material 1=NamedRerouteUsage()
    Height Texture 1=NamedRerouteUsage()
    Material 2=NamedRerouteUsage()
    Height Texture 2=NamedRerouteUsage()
    Transistion Phase=NamedRerouteUsage()
   ) {materialFunction=&MF_HeightBlendLandscapeLayers}.BlendedMaterial) {name=RVT-Material_1, nodeColor=<rgba 1 0.5391 0 1>}
  n14 = LandscapeLayerCoords() {mappingScale=2033}
  n15 = TextureSample(UVs=NamedRerouteDeclaration(None=StaticSwitch(True=@n14, False=@n2.Adjustable)) {name=BaseCoord, nodeColor=<rgba 1 0 1 1>}) {samplerType=SAMPLERTYPE_LinearColor, texture=/Game/Athena/Apollo/Environments/Landscape/Materials/Biome/CommonMaterials/T_SnowMask_Static}
  NamedRerouteDeclaration(None=Multiply(A=@n15.R, B=@n15.R)) {name=SnowBiomeMask, nodeColor=<rgba 0 0.4141 1 1>}
  NamedRerouteDeclaration(None=FunctionInput(Preview=Constant()) {bUsePreviewValueAsDefault=true, inputName=OpacityMask, inputType=FunctionInput_Scalar, sortPriority=-400}) {name=OpacityMask, nodeColor=<rgba 0.5938 0 1 1>}
  NamedRerouteDeclaration(None=FunctionInput(Preview=Constant()) {bUsePreviewValueAsDefault=true, inputName=RoadTypeEdgeFalloffMask, inputType=FunctionInput_Scalar, sortPriority=-300}) {name=RoadTypeEdgeFalloffMask, nodeColor=<rgba 0.5938 0 1 1>}
  n16 = ScalarParameter() {defaultValue=-0.25, group=HeliosSnow, parameterName=RoadSnow__Noise_Min}
  n17 = ScalarParameter() {defaultValue=1, group=HeliosSnow, parameterName=RoadSnow_Noise_Max}
  n18 = TextureSampleParameter2D(UVs=Divide(A=WorldPosition().XY, B=ScalarParameter() {defaultValue=4096, group=HeliosSnow, parameterName=RoadSnow_Noise_Scale})) {group=HeliosSnow, parameterName=RoadSnow_CoverageNoise, texture=/Game/Effects/Fort_Effects/Textures/Tile/T_GaussianNoise_01_M}
  n19 = ScalarParameter() {defaultValue=1, group=HeliosSnow, parameterName=RoadSnow_Noise_Strength}
  NamedRerouteDeclaration(None=Multiply(A=LinearInterpolate(A=@n16, B=@n17, Alpha=@n18.R), B=@n19)) {name=SnowMask, nodeColor=<rgba 0 0.7656 1 1>}
  NamedRerouteDeclaration(None=Saturate(None=@n13.OpacityMask)) {name=Snow_Height, nodeColor=<rgba 1 1 1 1>}
  NamedRerouteDeclaration(None=@n10.OpacityMask) {name=InputMaterial_Height, nodeColor=<rgba 0 1 0 1>}
  n20 = Constant() {r=0.005}
  MaterialExpressionReroute_1 = Reroute(None=@n2.Adjustable)
  n21 = Abs(None=Subtract(A=@MaterialExpressionReroute_1.InUVs) {constB=0.5})
  n22 = MaterialFunctionCall(Float2=OneMinus(None=Multiply(A=@n21) {constB=2})) {materialFunction=&BreakOutFloat2Components}
  MaterialExpressionReroute_3 = Reroute(None=@n1)
  n23 = StaticSwitch(
   True=Multiply(A=Step(Y=@n20, X=@n22.R) {constY=0.01}, B=Step(Y=@n20, X=@n22.G) {constX=0.001, constY=0.01})
   False=Constant() {r=1}
   Value=@MaterialExpressionReroute_3.Enabled
  )
  MaterialExpressionReroute_0 = Reroute(None=@n23)
  NamedRerouteDeclaration(None=@MaterialExpressionReroute_0.Output) {name=PageMask, nodeColor=<rgba 0.875 0 1 1>}
  n24 = ScalarParameter() {defaultValue=2, group=HeliosSnow, parameterName=RoadSnow_RoadRadius}
  n25 = SphereMask(
   A=NamedRerouteUsage()
   B=Constant() {r=1}
   Radius=@n24
   Hardness=ScalarParameter() {group=HeliosSnow, parameterName=RoadSnow_RoadEdgeHardness}
  )
  NamedRerouteDeclaration(None=@n25) {name=RVTMask, nodeColor=<rgba 1 0.1172 0 1>}
  n26 = Multiply(A=@n25, B=ScalarParameter() {defaultValue=1.4, group=HeliosSnow, parameterName=RoadSnow_GradientStrength})
  n27 = NamedRerouteUsage()
  n28 = Multiply(A=Saturate(None=Subtract(A=@n26, B=@n27)), B=NamedRerouteUsage())
  NamedRerouteDeclaration(None=@n28) {name=RoadEdgeMask, nodeColor=<rgba 0 0.1094 1 1>}
  NamedRerouteDeclaration(None=@n5.BlendedMaterial) {name=NoRVT-Material, nodeColor=<rgba 0.0781 0 1 1>}
  PinBase!(Output=@n23) {reroutePins [Output {expression=&self:MaterialExpressionReroute_0}]}
  PinBase!() {reroutePins [
    InUVs {expression=&self:MaterialExpressionReroute_1}
    Enabled {expression=&self:MaterialExpressionReroute_3}
   ]}
  Reroute(None=MaterialProxyReplace(Realtime=@n10.OpacityMask, MaterialProxy=Constant() {desc="Previously \"0\"", r=1}))
  Reroute(None=MaterialProxyReplace(
    Realtime=@n19
    MaterialProxy=ScalarParameter() {defaultValue=1, desc=Previously 0.1, group=HeliosSnow, parameterName=RoadSnow_Noise_ProxyStrength}
   ))
  Reroute(None=MaterialProxyReplace(
    Realtime=@n24
    MaterialProxy=ScalarParameter() {defaultValue=2, desc=Previously 3.5, group=HeliosSnow, parameterName=RoadSnow_ProxyRoadRadius}
   ))
  n29 = NamedRerouteUsage()
  n30 = Multiply(A=@n29, B=@n29)
  Saturate(None=Add(A=OneMinus(None=@n30), B=@n28))
  StaticSwitchParameter(True=MaterialProxyReplace(Realtime=@n28, MaterialProxy=Multiply(A=@n29, B=@n28)), False=@n28) {defaultValue=true, group=HeliosSnow, parameterName=Use_SnowBiomeMaskTexture, sortPriority=-800}
 ]}
