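# Apollo landscape material function that builds an asphalt road surface with cracks, outputting material attributes for landscape, mobile and static mesh use plus crack height and normal masks.
# The default path uses world-aligned Z projections of detail asphalt and crack textures, darkening base color with the crack mask and height-lerping crack height against detail spec.
# The mobile path samples one packed texture with world UVs to derive color, specular, roughness and crack height, using vertex normals.
# Shading path switches select between them.
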MaterialFunction /Game/Athena/Apollo/Environments/Landscape/Materials/MaterialFunctions/MF_ApolloAsphalt
graph {nodes [
  n1 = NamedRerouteUsage()
  FunctionOutput(None=ShadingPathSwitch(Default=NamedRerouteUsage(), Forward=@n1, Mobile=@n1))
  n2 = StaticBool()
  n3 = StaticBool() {value=true}
  n4 = MaterialFunctionCall!(
   Diffuse Texture Object=TextureObject() {samplerType=SAMPLERTYPE_Masks, texture=/Game/Athena/Apollo/Environments/Landscape/Textures/Roads/Road_Cracks_M}
   Spec Texture Object
   TextureSize=NamedRerouteUsage()
   --------------------------
   Use seperate settings for Z projection
   Use seperate texture for Z projection
   Texture Size Z projection
   Z Plane Diffuse Texture Object
   Z Plane Spec Texture Object
   "--------------------------      "
   ProjectionTransitionContrast
   Use Texture Alpha Heightmap to Refine the Transistion=@n2
   HeightLerpTransitionContrast
   PickTheDiffuseHeightChannel
   TheDiffuseHeightChannel
   "--------------------------       "
   NormalTextureObject=TextureObject() {samplerType=SAMPLERTYPE_Normal, texture=/Game/Athena/Apollo/Environments/Landscape/Textures/Roads/Road_Cracks_N}
   World Space Normals=@n3
   "--------------------------  "
   Export Float 4 Diffuse=@n2
   Export Float 4 Spec
   "--------------------------     "
   WorldPosition=WorldPosition() {worldPositionShaderOffset=WPT_ExcludeAllShaderOffsets}.XYZ
   World Space Normal
  ) {materialFunction=/Engine/Functions/Engine_MaterialFunctions02/Texturing/WorldAlignedTextures_Complex}
  FunctionOutput(None=Multiply(A=@n4.Normal Z Texture, B=Constant() {r=8})) {outputName=CrackNormal, sortPriority=8}
  n5 = NamedRerouteUsage()
  FunctionOutput(None=ShadingPathSwitch(Default=NamedRerouteUsage(), Forward=@n5, Mobile=@n5)) {outputName=CrackHeight, sortPriority=4}
  n6 = TextureSample(UVs=MaterialFunctionCall!(Scale=NamedRerouteUsage()) {materialFunction=/Game/Athena/Apollo/Environments/Landscape/Materials/Biome/CommonMaterials/MF_SimpleWorldUV}.Result) {
   samplerSource=SSM_Wrap_WorldGroupSettings
   samplerType=SAMPLERTYPE_Masks
   texture=/Game/Athena/Apollo/Environments/Landscape/Textures/Roads/T_Apollo_Asphalt_MOBILE
  }
  n7 = TextureSample(UVs=MaterialFunctionCall(Scale=NamedRerouteUsage()) {materialFunction=&MF_SimpleWorldUV}.Result) {
   samplerSource=SSM_Wrap_WorldGroupSettings
   samplerType=SAMPLERTYPE_Masks
   texture=&T_Apollo_Asphalt_MOBILE
  }
  n8 = OneMinus(None=Multiply(A=NamedRerouteUsage(), B=@n7.R))
  n9 = Multiply(A=Multiply(A=Constant3Vector() {constant=<rgba 0.1 0.0985 0.085 1>}, B=@n6.B), B=@n8)
  n10 = Multiply(A=@n8, B=Constant() {r=0.5})
  n11 = OneMinus(None=@n6.B)
  n12 = Constant() {r=0.2}
  n13 = Constant() {r=-0.2}
  n14 = Saturate(None=Add(A=Add(A=@n11, B=@n12), B=Add(A=@n13, B=@n7.B)))
  n15 = SetMaterialAttributes!(MaterialAttributes, BaseColor=@n9, Specular=@n10, Roughness=@n14, Normal=VertexNormalWS()) {attributeSetTypes [
    #1
    #2
    #3
    #4
   ]}
  FunctionOutput(None=@n15) {outputName=MobileLandscapeMaterial}
  n16 = NamedRerouteUsage()
  FunctionOutput(None=ShadingPathSwitch(Default=NamedRerouteUsage(), Forward=@n16, Mobile=@n16)) {outputName=ForStaticmesh}
  n17 = NamedRerouteUsage()
  FunctionOutput(None=ShadingPathSwitch(Default=NamedRerouteUsage(), Forward=@n17, Mobile=@n17)) {outputName=SimpleCrackHeight, sortPriority=4}
  n18 = GetMaterialAttributes(None=NamedRerouteUsage()) {attributeGetTypes [#4]}
  FunctionOutput(None=SetMaterialAttributes!(
    MaterialAttributes=@n18.MaterialAttributes
    Normal=Transform(None=@n18.Normal) {transformSourceType=TRANSFORMSOURCE_World, transformType=TRANSFORM_Tangent}
   ) {attributeSetTypes [#4]}) {outputName=DefaultLandscapeMaterial}
  NamedRerouteDeclaration(None=Constant() {r=2000}) {name=CrackSize, nodeColor=<rgba 0.5469 0 1 1>}
  NamedRerouteDeclaration(None=FunctionInput(Preview=Constant() {r=350}) {bUsePreviewValueAsDefault=true, inputName=TilingSize, inputType=FunctionInput_Scalar}) {name=AsphaltScale, nodeColor=<rgba 0.5469 0 1 1>}
  NamedRerouteDeclaration(None=Constant() {r=0.5905}) {name=AsphaltVariation, nodeColor=<rgba 0.5469 0 1 1>}
  NamedRerouteDeclaration(None=@n7.G) {name=Mobile-CrackHeight, nodeColor=<rgba 0.5469 0 1 1>}
  n19 = ComponentMask(None=@n4.Z Diffuse Texture) {g=true}
  n20 = StaticBool()
  n21 = MaterialFunctionCall(
   Diffuse Texture Object=TextureObject() {texture=/Game/Athena/Apollo/Environments/Landscape/Textures/Roads/T_AsphaltDetail1_D}
   Spec Texture Object=TextureObject() {samplerType=SAMPLERTYPE_Masks, texture=/Game/Athena/Apollo/Environments/Landscape/Textures/Roads/T_AsphaltDetail1_S}
   TextureSize=NamedRerouteUsage()
   Use Texture Alpha Heightmap to Refine the Transistion=@n20
   NormalTextureObject=TextureObject() {samplerType=SAMPLERTYPE_Normal, texture=/Game/Athena/Apollo/Environments/Landscape/Textures/Roads/T_AsphaltDetail1_N}
   World Space Normals=@n3
   Export Float 4 Diffuse=@n20
   WorldPosition=WorldPosition() {worldPositionShaderOffset=WPT_ExcludeAllShaderOffsets}.XYZ
  ) {materialFunction=&WorldAlignedTextures_Complex}
  n22 = MaterialFunctionCall(Float3=@n21.Z Spec) {materialFunction=&BreakOutFloat3Components}
  n23 = Multiply(A=OneMinus(None=@n22.B), B=Constant() {r=1.35}) {constB=2}
  NamedRerouteDeclaration(None=MaterialFunctionCall(Transition Phase=@n19, Height Texture=Saturate(None=@n23)) {materialFunction=&HeightLerp}.Lerp Alpha No Contrast) {name=Default-CrackHeight, nodeColor=<rgba 0.5469 0 1 1>}
  n24 = Multiply(A=NamedRerouteUsage(), B=ComponentMask(None=@n4.Z Diffuse Texture) {r=true})
  n25 = SetMaterialAttributes!(
   MaterialAttributes
   BaseColor=Multiply(A=@n21.Z Diffuse Texture, B=OneMinus(None=@n24))
   Specular=@n22.R
   Roughness=@n22.R
   Normal=@n21.Normal Z Texture
   SubsurfaceColor=Constant4Vector().RGBA
   Opacity=Constant() {r=1}
   ShadingModel=ShadingModel()
  ) {attributeSetTypes [
    #1
    #2
    #3
    #4
    #5
    #6
    #7
   ]}
  NamedRerouteDeclaration(None=@n25) {name=Default-MaterialOut, nodeColor=<rgba 0.5469 0 1 1>}
  n26 = SetMaterialAttributes(
   BaseColor=@n9
   Specular=@n10
   Roughness=@n14
   Normal=VertexNormalWS()
   SubsurfaceColor=Constant4Vector().RGBA
   Opacity=Constant() {r=1}
   ShadingModel=ShadingModel()
  ) {attributeSetTypes [
    #1
    #2
    #3
    #4
    #5
    #6
    #7
   ]}
  NamedRerouteDeclaration(None=@n26) {name=Mobile-MaterialOut, nodeColor=<rgba 0.5469 0 1 1>}
  NamedRerouteDeclaration(None=@n19) {name=Default-SimpleCrackHeight, nodeColor=<rgba 0.5469 0 1 1>}
  ScalarParameter() {defaultValue=0.5905, group=Asphalt, parameterName=AsphaltVariation}
  ScalarParameter() {defaultValue=350, group=Asphalt, parameterName=AsphaltMainScale}
  ScalarParameter() {defaultValue=0.5, group=Asphalt, parameterName=MobileAsphaltSpec}
  ScalarParameter() {defaultValue=0.2, group=Asphalt, parameterName=MobileAsphaltRough}
  ScalarParameter() {defaultValue=-0.2, group=Asphalt, parameterName=MobileAsphaltRough2}
  ScalarParameter() {defaultValue=2000, group=Asphalt, parameterName=SizeCracks}
  TextureObjectParameter() {group=Asphalt, parameterName=Asphalt Diffuse, texture=&T_AsphaltDetail1_D}
  TextureObjectParameter() {
   group=Asphalt
   parameterName=Asphalt_SMR
   samplerType=SAMPLERTYPE_Masks
   texture=&T_AsphaltDetail1_S
  }
  TextureObjectParameter() {
   group=Asphalt
   parameterName=Asphalt Normal
   samplerType=SAMPLERTYPE_Normal
   texture=&T_AsphaltDetail1_N
  }
 ]}
