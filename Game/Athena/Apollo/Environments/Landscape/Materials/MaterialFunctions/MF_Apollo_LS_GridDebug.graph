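# Material function for the Apollo landscape that supplies a debug grid overlay, projected from world position.
# It offsets world position by a grid start vector, then divides by cell size, derived from landscape size over grid amount.
# The resulting XY serves as UVs for a grid debug texture, returned as separate RGB and Alpha outputs.
# Start position, landscape size and grid amount are exposed as parameters.
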MaterialFunction /Game/Athena/Apollo/Environments/Landscape/Materials/MaterialFunctions/MF_Apollo_LS_GridDebug
graph {nodes [
  n1 = Add(A=WorldPosition().XYZ, B=VectorParameter() {parameterName=GridStartPosition, defaultValue=<rgba 130048 -130048 0 0>}.RGB)
  n2 = Divide(A=ScalarParameter() {defaultValue=260096, parameterName=LandscapeSize}, B=ScalarParameter() {defaultValue=8, parameterName=GridAmount})
  n3 = TextureSample(UVs=ComponentMask(None=Divide(A=@n1, B=@n2)) {g=true, r=true}) {texture=/Game/Athena/Apollo/Environments/Landscape/Textures/T_Apollo_GridDebug_01}
  FunctionOutput(None=@n3.RGB) {outputName=RGB}
  FunctionOutput(None=@n3.A) {outputName=Alpha, sortPriority=1}
 ]}
